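; BP_PlanetTest is a prototype planet actor that sizes a planet mesh and configures its material from kilometre-based properties.
; Its construction script converts the radius to centimetres, scales the mesh assuming a 256-unit base diameter, and lowers it by one radius so the top sits at the actor origin.
; It assigns a dynamic material, feeds it radius, terrain elevation and ocean depth, and stores a radius-derived mesh scale.
; The BeginPlay, overlap and Tick events are empty.
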(event EventBeginPlay)

(event Collision|EventActorBeginOverlap
  (params (OtherActor :type Actor%20Object%20Reference)))

(event EventTick
  (params (DeltaSeconds :type Float%20%28single-precision%29)))

(fn ConstructionScript
  (bind _planet_radius_km (Variables|Default|GetPlanet%20Radius%20Km))
  (bind _returnvalue (* _planet_radius_km 1e+05))
  (bind _mid (Variables|Default|GetMID))
  (Class|VolumetricsBPLIbrary|CreateAndAssignMID (Variables|Default|GetMID) (Variables|Default|GetMaterial))
  (Transformation|SetWorldScale3D (Variables|Default|GetPlanetMesh) (Math|Conversions|ToVector(Float) (/ (* _returnvalue 2.0) 256.0)))
  (Transformation|SetRelativeLocation (Variables|Default|GetPlanetMesh) (Math|Vector|MakeVector 0.0 0.0 (* _returnvalue -1.0)) false false)
  (Rendering|Material|SetMaterial (Variables|Default|GetPlanetMesh) 0 (Variables|Default|GetMID))
  (Rendering|Material|SetScalarParameterValue (Variables|Default|GetMID) "PlanetRadiusKm" (Variables|Default|GetPlanet%20Radius%20Km) :!class /Script/Engine.MaterialInstanceDynamic)
  (Rendering|Material|SetScalarParameterValue _mid "TerrainElevationKm" (Variables|Default|GetTerrain%20Elevation%20Km) :!class /Script/Engine.MaterialInstanceDynamic)
  (Rendering|Material|SetScalarParameterValue _mid "OceanDepthKm" (Variables|Default|GetOcean%20Depth%20Km) :!class /Script/Engine.MaterialInstanceDynamic)
  (Variables|Default|SetMesh%20scale (/ (Variables|Default|GetPlanet%20Radius%20Km) 1.6e+02)))
