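# Material function for a planet prototype's vertex shader that re-orients world positions relative to the camera's direction from a planet center below the origin.
# PlanetRadiusKm, scaled to centimeters, supplies the Z shift to that center.
# CreateThirdOrthogonalVector derives vectors from that direction for Transform3x3Matrix, whose three basis inputs all read Vector2 as wired.
# Outputs are the centered position, the tangent-aligned position shifted back, and the resulting World Position Offset.
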MaterialFunction /Volumetrics/Content/PlanetPrototype/Functions/Planet_VertexShader
graph {nodes [
  n1 = MaterialFunctionCall(Z=Multiply(A=ScalarParameter() {parameterName=PlanetRadiusKm}) {constB=100000}) {materialFunction=&MakeFloat3}
  n2 = CameraPositionWS()
  n3 = MaterialFunctionCall!(Vector1=Normalize(VectorInput=Add(A=@n2, B=@n1.Result)), Vector2=Constant3Vector() {constant=<rgba 0.1 0 1 0>}) {materialFunction=/Engine/Functions/Engine_MaterialFunctions02/Math/CreateThirdOrthogonalVector}
  n4 = MaterialFunctionCall(
   VectorToTransform=Add(A=WorldPosition() {worldPositionShaderOffset=WPT_ExcludeAllShaderOffsets}.XYZ, B=@n1.Result)
   BasisX=@n3.Vector2
   BasisY=@n3.Vector2
   BasisZ=@n3.Vector2
  ) {materialFunction=&Transform3x3Matrix}
  FunctionOutput(None=@n4.Transformed Vector) {outputName=World Position - Centered}
  n5 = Subtract(A=@n4.Transformed Vector, B=@n1.Result)
  FunctionOutput(None=@n5) {outputName=World Position - Tangent Aligned, sortPriority=1}
  n6 = WorldPosition() {worldPositionShaderOffset=WPT_ExcludeAllShaderOffsets}
  FunctionOutput(None=Add(A=Multiply(A=@n6.XYZ) {constB=-1}, B=@n5)) {outputName=World Position Offset, sortPriority=2}
 ]}
