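# Material function for planet water that outputs a world normal rippled by two panning layers of a tiling water normal texture.
# Each layer samples world-position XY at its own scale, offset by time along a parameterized direction and speed.
# The layers are averaged, remapped to signed range and given a derived Z.
# A strength parameter lerps the normalized WorldPosition input, serving as the sphere normal, toward that detail normal without tangent-space conversion or renormalization.
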MaterialFunction /Volumetrics/Content/PlanetPrototype/Functions/UnrealSky_Planet_WaterPanner
graph {nodes [
  n1 = FunctionInput(Preview=WorldPosition().XYZ) {bUsePreviewValueAsDefault=true, inputName=WorldPosition}
  n2 = MaterialFunctionCall!(Speed=ScalarParameter() {defaultValue=0.2, parameterName=Water 1 Panner}) {materialFunction=&TimeWithSpeedVariable}
  n3 = Normalize(VectorInput=VectorParameter() {parameterName=Panner 1, defaultValue=<rgba 0.4 0.2 0.9 0>}.RGB)
  n4 = Add(
   A=MaterialFunctionCall!(Scale=ScalarParameter() {defaultValue=67000, parameterName=Normal Size}) {materialFunction=/Engine/Functions/Engine_MaterialFunctions01/Coordinates/WorldPosition-XY}.Result
   B=ComponentMask(None=Multiply(A=@n2.Time, B=@n3)) {g=true, r=true}
  )
  n5 = MaterialFunctionCall(Speed=ScalarParameter() {defaultValue=0.1, parameterName=Water 2 Panner}) {materialFunction=&TimeWithSpeedVariable}
  n6 = Normalize(VectorInput=VectorParameter() {parameterName=Panner 2, defaultValue=<rgba -0.2 -0.8 0 0>}.RGB)
  n7 = Add(
   A=MaterialFunctionCall(Scale=ScalarParameter() {defaultValue=93400, parameterName=Normal 2 Size}) {materialFunction=&WorldPosition-XY}.Result
   B=ComponentMask(None=Multiply(A=@n5.Time, B=@n6)) {g=true, r=true}
  )
  n8 = ConstantBiasScale(None=LinearInterpolate(
    A=TextureSample(UVs=@n4) {samplerType=SAMPLERTYPE_LinearColor, texture=/Volumetrics/Content/PlanetPrototype/Textures/T_Water_TilingNormal_With_Height}.RGB
    B=TextureSample(UVs=@n7) {samplerType=SAMPLERTYPE_LinearColor, texture=&T_Water_TilingNormal_With_Height}.RGB
   )) {bias=-0.5, scale=2}
  FunctionOutput(None=LinearInterpolate(A=Normalize(VectorInput=@n1), B=DeriveNormalZ(InXY=@n8), Alpha=ScalarParameter() {defaultValue=0.04, parameterName=Water Normal Strength})) {outputName=World Normal}
 ]}
