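# Prototype planet surface material that builds procedural terrain and oceans on a Subsurface shading model.
# Ten-level value noise raises land radially through world position offset atop Planet_VertexShader and yields the land/water mask.
# The mask switches roughness and normals: panned water normals, or land normals from screen-space derivatives of displaced positions.
# World-aligned textures vary biome colors, while exponential depth extinction drives underwater attenuation and sky-lit scattering.
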Material /Volumetrics/Content/PlanetPrototype/Materials/M_Planet_01
graph {
 properties {bTangentSpaceNormal=false, pixelDepthOffsetMode=PDOM_Legacy, shadingModel=MSM_Subsurface, usedShadingModels=Subsurface}
 nodes [
  n1 = ScalarParameter() {parameterName=OceanDepthKm}
  n2 = MaterialFunctionCall() {materialFunction=/Volumetrics/Content/PlanetPrototype/Functions/Planet_VertexShader}
  n3 = ScalarParameter() {defaultValue=6234186, parameterName=g}
  n4 = LinearInterpolate(
   A=Divide(A=Multiply(A=@n1) {constB=-1}, B=ScalarParameter() {defaultValue=10, parameterName=TerrainElevationKm})
   B=ScalarParameter() {defaultValue=1, parameterName=max}
   Alpha=Noise!(World Position=Divide(A=@n2.World Position - Tangent Aligned, B=@n3), FilterWidth) {bTurbulence=false, levels=10, noiseFunction=NOISEFUNCTION_ValueALU, outputMin=0}
  )
  n5 = Max(A=@n4) {constB=0}
  n6 = Multiply(
   A=Multiply(A=Multiply(A=@n5, B=@n5), B=Normalize(VectorInput=@n2.World Position - Centered))
   B=Multiply(A=ScalarParameter() {defaultValue=10, parameterName=TerrainElevationKm}) {constB=100000}
  )
  Add(A=@n6, B=@n2.World Position Offset)
  n7 = Multiply(A=@n4)
  n8 = Saturate(None=Ceil(None=@n7))
  LinearInterpolate(A=ScalarParameter() {defaultValue=0.15, parameterName=Water R}, B=ScalarParameter() {defaultValue=1, parameterName=Land R}, Alpha=@n8)
  n9 = MaterialFunctionCall() {materialFunction=&Planet_VertexShader}
  n10 = Add(A=@n9.World Position - Tangent Aligned, B=@n6)
  n11 = DDX(Value=@n10)
  n12 = DDY(Value=@n10)
  LinearInterpolate(
   A=MaterialFunctionCall!(WorldPosition=@n9.World Position - Centered) {materialFunction=/Volumetrics/Content/PlanetPrototype/Functions/UnrealSky_Planet_WaterPanner}.World Normal
   B=Normalize(VectorInput=CrossProduct(A=@n11, B=@n12))
   Alpha=@n8
  )
  n13 = MaterialFunctionCall(
   TextureObject=TextureObject() {texture=/Engine/Engine_MI_Shaders/T_Base_Tile_Diffuse}
   TextureSize=ScalarParameter() {defaultValue=1650000, parameterName=Terrain Noise scale}
   WorldPosition=WorldPosition().XYZ
  ) {materialFunction=&WorldAlignedTexture}
  n14 = MaterialFunctionCall(
   TextureObject=TextureObject() {texture=/Engine/EngineMaterials/T_Default_MacroVariation}
   TextureSize=ScalarParameter() {defaultValue=7650000, parameterName=Dirt Noise Scale}
   WorldPosition=WorldPosition().XYZ
  ) {materialFunction=&WorldAlignedTexture}
  n15 = MaterialFunctionCall(
   TextureObject=TextureObject() {texture=/Engine/Engine_MI_Shaders/T_Base_Tile_Variation}
   TextureSize=ScalarParameter() {defaultValue=9650000, parameterName=Biome Variation Scale}
   WorldPosition=WorldPosition().XYZ
  ) {materialFunction=&WorldAlignedTexture}
  n16 = Saturate(None=LinearInterpolate(
    A=ScalarParameter() {defaultValue=-0.1, parameterName=VarMin}
    B=ScalarParameter() {defaultValue=8, parameterName=VarMax}
    Alpha=ComponentMask(None=@n15.Z Texture) {r=true}
   ) {constA=-0.5, constB=15})
  n17 = ScalarParameter() {defaultValue=4, parameterName=Var}
  n18 = MaterialFunctionCall!(Alpha=OneMinus(None=Saturate(None=@n7)), Offset=ScalarParameter() {defaultValue=0.85, parameterName=w}) {materialFunction=/Engine/Functions/Engine_MaterialFunctions01/Opacity/AlphaOffset}
  n19 = ScalarParameter() {defaultValue=2, parameterName=ftre}
  n20 = Saturate(None=Add(A=Multiply(A=@n16, B=@n17), B=Multiply(A=@n18.Result, B=@n19)))
  n21 = LinearInterpolate(
   A=MaterialFunctionCall(
    TextureObject=TextureObject() {texture=/Engine/Tutorial/SubEditors/TutorialAssets/T_soft_smoke}
    TextureSize=ScalarParameter() {defaultValue=850000, parameterName=Underwater Noise Scale}
    WorldPosition=WorldPosition().XYZ
   ) {materialFunction=&WorldAlignedTexture}.Z Texture
   B=LinearInterpolate(A=Multiply(A=@n13.Z Texture) {constB=7}, B=SquareRoot(None=@n14.Z Texture), Alpha=@n20)
   Alpha=@n8
  )
  n22 = VectorParameter() {parameterName=Peakm, defaultValue=<rgba 0.5 0.5 0.2 0.5>}
  n23 = VectorParameter() {parameterName=Edge, defaultValue=<rgba 0.6302 0.4094 0.1584 0.7>}
  n24 = LinearInterpolate(A=Multiply(A=@n22.RGB, B=@n22.RGB), B=Multiply(A=@n23.RGB, B=@n23.RGB), Alpha=@n20)
  MaterialExpressionMax_0 = Max(A=Multiply(A=@n7) {constB=-1}) {constB=0}
  n25 = VectorParameter() {parameterName=Exct, defaultValue=<rgba 250 5000 227000 50>}
  MaterialExpressionAppendVector_0 = AppendVector(A=Divide(A=ScalarParameter() {defaultValue=50000, parameterName=Density Scale}, B=@n25.RGB), B=@n25.A)
  n26 = Custom!(t=@MaterialExpressionMax_0, c=@MaterialExpressionAppendVector_0) {
   code="return exp(-t * c);"
   description=Exponential Density
   outputType=CMOT_Float4
   inputs [
    {inputName=t, input {expression=&self:MaterialExpressionMax_0}}
    {inputName=c, input {expression=&self:MaterialExpressionAppendVector_0}}
   ]
  }
  n27 = ComponentMask(None=@n26) {a=true}
  n28 = Multiply(A=ComponentMask(None=@n26) {b=true, g=true, r=true}, B=@n27)
  LinearInterpolate(B=Multiply(A=Multiply(A=@n21, B=@n24), B=LinearInterpolate(A=@n28, Alpha=@n8) {constA=1}), Alpha=@n8) {constA=0.225}
  n29 = VectorParameter() {parameterName=Scattering, defaultValue=<rgba 0.2 0.3 0.2 0.03>}
  Multiply(A=OneMinus(None=@n27), B=Multiply(A=Multiply(A=@n29.RGB, B=@n29.RGB), B=SkyAtmosphereLightIlluminance!(World Position)))
 ]
}
