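# Planet water material for the PlanetPrototype, using the Single Layer Water shading model with non-tangent-space normals.
# Centered world position from Planet_VertexShader drives the UnrealSky_Planet_WaterPanner function.
# Water output takes scattering as color scaled by its alpha, absorption as the reciprocal of squared color, and anisotropy from a scalar.
# Scalar parameters R, O and S are also exposed.
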Material /Volumetrics/Content/PlanetPrototype/Materials/M_Planet_01_Water
graph {
 properties {bTangentSpaceNormal=false, pixelDepthOffsetMode=PDOM_Legacy, shadingModel=MSM_SingleLayerWater, usedShadingModels=SingleLayerWater}
 nodes [
  Constant()
  MaterialFunctionCall!(WorldPosition=MaterialFunctionCall() {materialFunction=/Volumetrics/Content/PlanetPrototype/Functions/Planet_VertexShader}.World Position - Centered) {materialFunction=/Volumetrics/Content/PlanetPrototype/Functions/UnrealSky_Planet_WaterPanner}
  ScalarParameter() {defaultValue=0.05, parameterName=R}
  ScalarParameter() {parameterName=O}
  ScalarParameter() {defaultValue=0.225, parameterName=S}
  n1 = VectorParameter() {group=Water Shading Model, parameterName=Scattering, defaultValue=<rgba 0.1 0.1 0.1 0.5>}
  n2 = VectorParameter() {group=Water Shading Model, parameterName=Absorption, defaultValue=<rgba 1 8 20 100>}
  SingleLayerWaterMaterialOutput!(
   ScatteringCoefficients=Multiply(A=@n1.RGB, B=Divide(A=@n1.A) {constB=100})
   AbsorptionCoefficients=Divide(B=Multiply(A=@n2.RGB, B=@n2.RGB))
   PhaseG=ScalarParameter() {defaultValue=0.5, parameterName=Aniso}
   ColorScaleBehindWater
  )
 ]
}
